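# Masked, lit popcorn material for the Corny character's effects, usable on Niagara mesh particles.
# Base color, normal and RMA textures supply the surface, with base color times particle color feeding an emissive adjustment function.
# The dynamic "selector" parameter snaps to thirds and is sphere-masked against vertex color red, isolating tagged mesh parts, likely popcorn variants.
# A DEBUG static switch substitutes material parameters for particle color and selector.
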Material /BRCosmetics/Characters/Player/Male/Medium/Bodies/M_MED_Corny/CornyFXs/M_Conry_PopcornFX
graph {
 properties {
  bUsedWithMorphTargets=true
  bUsedWithNiagaraMeshParticles=true
  blendMode=BLEND_Masked
  pixelDepthOffsetMode=PDOM_Legacy
  usedShadingModels=Default Lit
 }
 nodes [
  n1 = TextureSample() {texture=&/T_Corny_Popcorn_BC}
  n2 = VectorParameter() {parameterName=Debug Particle Color}
  n3 = ParticleColor()
  MaterialFunctionCall(Emissive=Multiply(A=@n1.RGB, B=StaticSwitchParameter(True=@n2.RGB, False=@n3.RGB) {parameterName=DEBUG})) {materialFunction=&MF_FX_EmissiveAdjustments}
  n4 = Add(A=StaticSwitchParameter(
    True=ScalarParameter() {parameterName=Debug Dynamic Parameter}
    False=DynamicParameter() {paramNames [selector,Param2,Param3,Param4]}.selector
   ) {parameterName=DEBUG}) {constB=0.01}
  n5 = Divide(A=Ceil(None=Multiply(A=@n4) {constB=3})) {constB=3}
  Reroute(None=SphereMask(A=VertexColor().R, B=Saturate(None=@n5), Radius=Constant() {r=0.1}))
  Reroute(None=@n1.RGB)
  Reroute(None=TextureSample() {samplerType=SAMPLERTYPE_Normal, texture=&/T_Corny_Popcorn_N}.RGB)
  TextureSample() {texture=&/T_Corny_Popcorn_RMA}
 ]
}
