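# Niagara burst effect of popcorn popping for the Corny player body cosmetic, made of popcorn meshes, smoke sprites and GPU dust sprites.
# Meshes launch from a stretched hemisphere, align to velocity, pop with squash-and-stretch scale curves, then fall and shrink away.
# Smoke expands and fades, while dust drifts on curl noise.
# User.Switch lerps mesh and smoke counts, dropping smoke to zero at 0, and User.spawnRatePopC adds to the dust count.
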NiagaraSystem /BRCosmetics/Characters/Player/Male/Medium/Bodies/M_MED_Corny/CornyFXs/NS_Corny_PopcornPop
dependencies {usedDataInterfaces [&NiagaraDataInterfaceCurve], usedDynamicInputs [], usedModules [&SystemState]}
emitters {
 Dust {
  simTarget=GPUComputeSim
  emitterUpdateScript [
   &EmitterState(Life Cycle Mode=::NewEnumerator1, Loop Behavior=::NewEnumerator1, Loop Duration=1, UseLoopDelay=false)
   &SpawnBurst_Instantaneous(Spawn Count=&Add_Integer(A=$User.spawnRatePopC, B=35), Use Spawn Probability=false)
  ]
  particleSpawnScript [
   /Niagara/Modules/Spawn/Initialization/InitializeParticle(
    Lifetime=&RandomRangeFloat(Minimum=0.8, Maximum=1.4)
    Mass=&RandomRangeFloat(Minimum=0.75, Maximum=1.75)
    Module.Write Color=true
    Module.Write Lifetime=true
    Module.Write Mass=true
    Module.Write Material Random=true
    Module.Write Position=false
    Module.Write Scale=false
    Module.Write SpriteRotation=true
    Module.Write SpriteSize=true
    Sprite Rotation=&RandomRangeFloat(Minimum=-180, Maximum=180)
    Sprite Size=&Vector2DFromFloat(Value=&RandomRangeFloat(Minimum=0.5, Maximum=2))
   )
   &AddVelocityInCone(
    Cone Angle=90
    Module.Use Velocity Falloff On Cone Axis=true
    Velocity Falloff Away From Cone Axis=0.333
    Velocity Strength=&RandomRangeFloat(Minimum=600, Maximum=1500)
   )
   &Color(Color=&RandomRangeLinearColor(Minimum=<rgba 1 0.7481 0.2744 1>, Maximum=<rgba 1 0.7481 0.2744 1>))
  ]
  particleUpdateScript [
   &ParticleState()
   &GravityForce(Gravity=<0 0 100>)
   &Drag(Drag=8, Use Linear Drag=true, Use Rotational Drag=false)
   &ScaleColor(
    Scale Alpha=&FloatFromCurve(FloatCurve=DataInterface() {propertyValues {Curve {keys [<at -0.0046 1.0113 RCIM_Cubic>,<at 1.0011 0.002 RCIM_Cubic>]}}})
    ScaleA=true
    ScaleRGB=true
   )
   /Niagara/Modules/Update/Forces/CurlNoiseForce(Noise Strength=90, Pan Noise=false)
   &SolveForcesAndVelocity(Clamp Velocity=false, Limit Acceleration=false)
   &SpriteRotationRate(Rotation Rate=&RandomRangeFloat(Minimum=-200, Maximum=200))
  ]
  rendererClasses [&NiagaraSpriteRendererProperties]
  renderers [{rendererClass=&NiagaraSpriteRendererProperties}]
 }
 NE_Corny_PopcornPop {
  emitterUpdateScript [
   &EmitterState(Life Cycle Mode=::NewEnumerator1, Loop Behavior=::NewEnumerator1, Loop Duration=1, UseLoopDelay=false)
   &SpawnBurst_Instantaneous(
    Spawn Count=&MakeIntFromFloat(Float=&Lerp_Float(A=15, B=60, Alpha=$User.Switch))
    Use Spawn Probability=false
   )
  ]
  particleSpawnScript [
   &InitializeParticle(
    Lifetime=&RandomRangeFloat(Minimum=0.5, Maximum=2)
    Mass=&RandomRangeFloat(Minimum=0.75, Maximum=1.75)
    Mesh Scale=&VectorFromFloat(Value=&RandomRangeFloat(Minimum=0.03, Maximum=0.07))
    Module.Write Color=true
    Module.Write Lifetime=true
    Module.Write Mass=true
    Module.Write Material Random=true
    Module.Write Position=false
    Module.Write Scale=true
    Module.Write SpriteRotation=false
    Module.Write SpriteSize=false
   )
   /Niagara/Modules/Update/Mass/CalculateSizeAndRotationalInertiaByMass(Depth=0, Height=5, Module.Manually Enter Density=false)
   &SphereLocation(
    Hemisphere X=true
    MaskSpawnGroup=false
    Non Uniform Scale=<2.5 1 1>
    Override Seed=false
    Sphere Radius=15
    Surface Only=true
    UseNonUniformScale=true
    UseOffset=false
    UseSphereOrientation=false
   )
   &AddVelocity(Velocity=&RandomRangeVector(Minimum=<50 0 0>, Maximum=<500 1 1>))
   &AddVelocity(Velocity=&Multiply_VectorByFloat(
     Vector=&TransformVector(Vector=$Output.SphereLocation.SphereNormal, Source Space=::World, Destination Space=::Local)
     Float=&RandomRangeFloat(Minimum=50, Maximum=350)
    )) {module=AddVelocity001}
   /Niagara/Modules/Spawn/Orientation/V2/InitialMeshOrientation(
    Rotation=&RandomRangeVector(Minimum=<-1 -1 -1>, Maximum=<1 1 1>)
    Use Orientation Vector=true
    Use Rotation Vector=true
   ) {enabled=false}
   &AddRotationalVelocity(Rotation Rate=&RandomRangeVector(Minimum=<-0.7 -0.7 -0.7>, Maximum=<0.7 0.7 0.7>))
   &DynamicMaterialParameters(
    Index 0 Param 1=&RandomRangeFloat(Minimum=0, Maximum=1)
    Module.Param0WriteEnabled=true
    Module.Param1WriteEnabled=false
    Module.Param2WriteEnabled=false
    Module.Param3WriteEnabled=false
   )
   &InitialMeshOrientation(
    Orientation Coordinate Space:ENiagaraCoordinateSpace=::World
    Orientation Vector=$Particles.Velocity
    Use Orientation Vector=true
    Use Rotation Vector=false
   ) {module=InitialMeshOrientation001}
  ]
  particleUpdateScript [
   &ParticleState()
   &ScaleMeshSize(Scale Factor=&MakeVector(
     X=&FloatFromCurve(FloatCurve=DataInterface() {propertyValues {Curve {keys [
          {arriveTangent=26.0828, interpMode=RCIM_Cubic, leaveTangent=26.0828, tangentMode=RCTM_User, time=0, value=0.2}
          {arriveTangent=0.0334, interpMode=RCIM_Cubic, leaveTangent=0.0334, tangentMode=RCTM_User, time=0.0394, value=4.0724}
          {arriveTangent=-2.5074, interpMode=RCIM_Cubic, leaveTangent=-2.5074, tangentMode=RCTM_User, time=0.1173, value=0.7442}
          {arriveTangent=-0.1609, interpMode=RCIM_Cubic, leaveTangent=-0.1609, tangentMode=RCTM_User, time=0.2152, value=1.1946}
          {arriveTangent=0.1391, interpMode=RCIM_Cubic, leaveTangent=0.1391, tangentMode=RCTM_User, time=0.3123, value=0.8985}
          <at 0.35 1>
         ]}}})
     Y=&FloatFromCurve(FloatCurve=DataInterface() {propertyValues {Curve {keys [
          <at -0.003 1.0812 RCIM_Cubic>
          {arriveTangent=3.927, interpMode=RCIM_Cubic, leaveTangent=3.927, time=0.0492, value=1.1749}
          {arriveTangent=-2.0352, interpMode=RCIM_Cubic, leaveTangent=-2.0352, time=0.1219, value=1.5716}
          {arriveTangent=-0.5182, interpMode=RCIM_Cubic, leaveTangent=-0.5182, tangentMode=RCTM_User, time=0.2148, value=0.8379}
          {arriveTangent=-0.6419, interpMode=RCIM_Cubic, leaveTangent=-0.6419, tangentMode=RCTM_User, time=0.3121, value=1.1071}
          <at 0.35 1 RCIM_Cubic>
         ]}}})
     Z=&FloatFromCurve(FloatCurve=DataInterface() {propertyValues {Curve {keys [
          <at -0.0016 1.0762 RCIM_Cubic>
          {arriveTangent=-0.5131, interpMode=RCIM_Cubic, leaveTangent=-0.5131, tangentMode=RCTM_User, time=0.0502, value=1.1706}
          {arriveTangent=-2.0432, interpMode=RCIM_Cubic, leaveTangent=-2.0432, time=0.1212, value=1.571}
          {arriveTangent=0.2115, interpMode=RCIM_Cubic, leaveTangent=0.2115, tangentMode=RCTM_User, time=0.2145, value=0.835}
          {arriveTangent=-0.2737, interpMode=RCIM_Cubic, leaveTangent=-0.2737, tangentMode=RCTM_User, time=0.312, value=1.1039}
          <at 0.35 1>
         ]}}})
    )) {module=ScaleMeshSize001}
   &GravityForce(Gravity=&RandomRangeVector(Minimum=<0 0 -500>, Maximum=<1 1 -1100>))
   &Drag(Drag=&RandomRangeFloat(Minimum=1, Maximum=3), Use Linear Drag=true, Use Rotational Drag=false)
   &ScaleColor(
    Scale Alpha=&FloatFromCurve(FloatCurve=DataInterface() {propertyValues {Curve {keys [<at -0.0046 1.0113 RCIM_Cubic>,<at 1.0011 0.002 RCIM_Cubic>]}}})
    Scale RGB=<0 0 0>
    ScaleA=true
    ScaleRGB=true
   )
   &SolveForcesAndVelocity(Clamp Velocity=false, Limit Acceleration=false)
   &SolveRotationalForcesAndVelocity()
   &ScaleMeshSize(Scale Factor=&VectorFromFloat(Value=&FloatFromCurve(FloatCurve=DataInterface() {propertyValues {Curve {keys [<at 0.75 1 RCIM_Cubic>, {arriveTangent=-8.6407, interpMode=RCIM_Cubic, leaveTangent=-8.6407, tangentMode=RCTM_User, time=1, value=0}]}}})))
  ]
  rendererClasses [&NiagaraMeshRendererProperties]
  renderers [{rendererClass=&NiagaraMeshRendererProperties, rendererIndex=0}]
 }
 Smoke_Burst {
  emitterUpdateScript [
   &EmitterState(Life Cycle Mode=::NewEnumerator1, Loop Behavior=::NewEnumerator1, Loop Duration=1, UseLoopDelay=false)
   &SpawnBurst_Instantaneous(
    Spawn Count=&MakeIntFromFloat(Float=&Lerp_Float(A=0, B=7, Alpha=$User.Switch))
    Use Spawn Probability=false
   )
  ]
  particleSpawnScript [
   &InitializeParticle(
    Lifetime=&RandomRangeFloat(Minimum=0.5, Maximum=1.5)
    Mass=&RandomRangeFloat(Minimum=0.75, Maximum=1.75)
    Module.Write Color=true
    Module.Write Lifetime=true
    Module.Write Mass=true
    Module.Write Material Random=true
    Module.Write Position=false
    Module.Write Scale=false
    Module.Write SpriteRotation=true
    Module.Write SpriteSize=true
    Sprite Rotation=&RandomRangeFloat(Minimum=-180, Maximum=180)
    Sprite Size=&Multiply_Vector2(
     A=<35 35>
     B=&Vector2DFromFloat(Value=&RandomRangeFloat(Minimum=0.5, Maximum=1.2))
    )
   )
   &AddVelocityInCone(
    Cone Angle=90
    Module.Use Velocity Falloff On Cone Axis=false
    Velocity Falloff Away From Cone Axis=0.333
    Velocity Strength=&RandomRangeFloat(Minimum=175, Maximum=350)
   )
  ]
  particleUpdateScript [
   &ParticleState()
   &Drag(Drag=5, Use Linear Drag=true, Use Rotational Drag=false)
   &ScaleColor(
    Scale Alpha=&Multiply_Float(
     A=0.2
     B=&FloatFromCurve(FloatCurve=DataInterface() {propertyValues {Curve {keys [<at 0 1 RCIM_Cubic>,<at 1 0>]}}})
    )
    Scale RGB=&MakeVectorFromLinearColorRGB(Color=<rgba 1 0.6845 0.394 1>)
    ScaleA=true
    ScaleRGB=true
   )
   &SolveForcesAndVelocity(Clamp Velocity=false, Limit Acceleration=false)
   &ScaleSpriteSize(Scale Factor=&Vector2DFromFloat(Value=&FloatFromCurve(FloatCurve=DataInterface() {propertyValues {Curve {keys [{arriveTangent=1.5974, interpMode=RCIM_Cubic, leaveTangent=1.5974, tangentMode=RCTM_User, time=0, value=0.8972}, <at 1 2 RCIM_Cubic>]}}})))
   &SpriteRotationRate(Rotation Rate=&RandomRangeFloat(Minimum=-100, Maximum=100))
  ]
  rendererClasses [&NiagaraSpriteRendererProperties]
  renderers [{rendererClass=&NiagaraSpriteRendererProperties}]
 }
}
summary {
 systemName=NS_Corny_PopcornPop
 emitters [
  {emitterName=NE_Corny_PopcornPop, rendererClasses [&NiagaraMeshRendererProperties]}
  {emitterName=Smoke_Burst, rendererClasses [&NiagaraSpriteRendererProperties]}
  {emitterName=Dust, simTarget=GPUComputeSim, rendererClasses [&NiagaraSpriteRendererProperties]}
 ]
 userVariables [
  User.spawnRatePopC {type=NiagaraInt32}
  User.Switch {type=NiagaraFloat, defaultValue=1, description=""}
 ]
}
